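; Living World Prefab that runs a rift-style encounter: an anticipation phase, then a burst that reveals its level and enables the spawner.
; The server waits out the delay, optionally destroys buildings in bounds, spawns burst audio, pushes players back and makes the level visible.
; Clients play anticipation VFX, shake and rumble, then lingering VFX, decal and physics force.
; A no-build box covers the bounds until the visibility RepNotify removes it.
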(event EventBeginPlay
  (bind (_origin _boxextent) (Collision|GetActorBounds true true))
  (bind _nobuildzone (Variables|Default|GetNoBuildZone))
  (Utilities|IsValid _nobuildzone
    (:Is%20Valid
      (Components|Box|SetBoxExtent _nobuildzone _boxextent true))
    (:Is%20Not%20Valid
      _))
  (Networking|SwitchHasAuthority
    (:Authority
      ;[1 1 1] Start the Anticipation
      (Utilities|FlowControl|Sequence
        (:then_0
          (Utilities|FlowControl|Delay 0.5)
          (Variables|Default|SetAllowVFXToSpawn false))
        (:then_1
          (CallFunction|StartEncounter))))
    (:Remote
      (Utilities|FlowControl|Sequence
        (:then_0
          (if (Variables|Default|GetAllowVFXToSpawn)
            (CallFunction|SpawnAncitipationVFX)))
        (:then_1
          ;[1 1 1] Make sure Lingering VFX is spawned
          (if (Variables|Default|GetLevelIsVisble)
            (if (Variables|Default|GetAllowLingerVFXToSpawn)
              (Utilities|IsValid (Variables|InternalData|GetHandle_AmbientVFX)
                (:Is%20Not%20Valid
                  (CallFunction|SpawnLingeringVFX))))))))))

(event EventEndPlay
  (params (EndPlayReason :type EEndPlayReason%20Enum :default "Destroyed"))
  (bind _spawnedanticipationvfx (Variables|InternalData|GetSpawnedAnticipationVFX))
  (bind _handle_ambientvfx (Variables|InternalData|GetHandle_AmbientVFX))
  ;[1 1 1] Make sure to remove the ForceFeedback on client in case client is out of relevancy
  (Networking|SwitchHasAuthority
    (:Remote
      (Components|Activation|Deactivate (Variables|Default|GetForceFeedback))
      (Utilities|IsValid _handle_ambientvfx
        (:Is%20Valid
          (Components|DestroyComponent _handle_ambientvfx))
        (:Is%20Not%20Valid
          _))
      (Utilities|IsValid _spawnedanticipationvfx
        (:Is%20Valid
          (Components|Activation|SetActive _spawnedanticipationvfx false false)
          (Components|DestroyComponent _spawnedanticipationvfx))))))

(event Custom|StartEncounter
  ;[1 1 1] Ancitipation Delay for Rift Encounter
  (if (not (and (Editor|InEditororPIE) (Variables|Default|GetDebugSkipAnticipationDelayinPIE)))
    (Utilities|FlowControl|Delay (ScalableFloat|GetValueatLevel (Variables|Default|GetAnticipationDelay))))
  (if (Variables|Default|GetbDestroyBuildingsInLevelBounds)
    (CallFunction|DestroyBuildingsinLevelBounds (delegate Custom|OnBuildingsDestroyed))
    (else
      (CallFunction|TriggerPreSpawnVFX)
      (CallFunction|SpawnAudioBurst)
      (CallFunction|PushBackPlayers)
      (Utilities|FlowControl|Delay 0.2)
      (CallFunction|MakeLevelVisible)
      (Variables|Default|SetLevelIsVisble true)
      (CallFunction|PostSpawnEvents))))

(event Custom|OnBuildingsDestroyed
  ;[1 1 1] Ancitipation Delay for Rift Encounter
  (CallFunction|TriggerPreSpawnVFX)
  (CallFunction|SpawnAudioBurst)
  (CallFunction|PushBackPlayers)
  (Utilities|FlowControl|Delay 0.2)
  (CallFunction|MakeLevelVisible)
  (Variables|Default|SetLevelIsVisble true)
  (CallFunction|PostSpawnEvents))

(event Custom|SpawnAudioBurst
  ;[1 1 1] Spawn a new Actor for the Audio Rift Encounter Open so Audio has a higher relevancy
  (Game|SpawnActorfromClass (Variables|Default|GetRiftBurstOpen) (Transformation|GetActorTransform self) "AlwaysSpawn" "SelectDefaultAtRuntime" self 0))

(event Custom|PushBackPlayers
  (bind _self self)
  (bind _returnvalue (Ability|GetAbilitySystemComponent _self))
  ;[1 1 1] Attempt to Push Back any Players that are nearby
  (Utilities|IsValid _returnvalue
    (:Is%20Valid
      (Ability|Advanced|GrantandActivateAbilityonSelf _returnvalue (Variables|Default|GetPushbackAbility) _self))))

(event Custom|StartAnticipationCameraShake
  (bind _camerashakesource (Variables|Default|GetCameraShakeSource))
  (bind _forcefeedback (Variables|Default|GetForceFeedback))
  ;[1 1 1] Enable Camera Shake and Controller Rumble
  (if (Variables|Default|GetAllowAncitipationCameraShake)
    (CameraShake|Start _camerashakesource)
    (Components|Activation|Activate _forcefeedback false)
    (Utilities|FlowControl|Delay (ScalableFloat|GetValueatLevel (Variables|Default|GetAnticipationDelay)))
    (CameraShake|StopAllCameraShakes _camerashakesource true)
    (Components|Activation|Deactivate _forcefeedback)))

(event Custom|TriggerPreSpawnVFX
  ;[1 1 1] Trigger VFX just before we spawn in the World Assets
  (Networking|SwitchHasAuthority
    (:Remote
      (CallFunction|SpawnLingeringVFX) ; Moved the Explosion VFX inside the SpawnAudioBurst \nEvent to have better timing.
      ;;[1 1 1] RadialForce
      (Utilities|AllowJanus
        (:Allow
          (bind _returnvalue (Development|GetConsoleVariableIntValue "r.PhysicsField.EnableField"))
          (if (== _returnvalue 1)
            (Game|SpawnActorfromClass "/Game/Effects/Niagara/Blueprints/PhsyicsFields/FN_RadialForce.FN_RadialForce_C" (Math|Transform|MakeTransform (+ (CallFunction|GetTeleportLocation) "0,0,-100.000000")) "Undefined" "MultiplyWithRoot" 0 0.2 0.75 3e+03 1.5 0.0 0)))))))

(event Custom|SpawnLingeringVFX
  (bind _self self)
  (bind _returnvalue (Transformation|GetActorLocation _self))
  (bind _spawn_decal_delay_duration (Variables|Default|GetSpawn%20Decal%20Delay%20Duration))
  ;[1 1 1] Spawn the Lingering VFX - Putting here in case we need to re-activate because of Relevency
  (SoundIndicator|StopSoundIndicatorLooping (Variables|Default|GetFortSoundIndicator))
  (if (Variables|Default|GetAllowLingerVFXToSpawn)
    (bind _returnvalue_2 (Niagara|SpawnSystematLocation (Variables|Default|GetFX_RiftSpawnIn_Linger) (Transformation|GetActorLocation self))) ; AmbientFX
    (Variables|InternalData|SetHandle_AmbientVFX _returnvalue_2)
    (else
      _))
  (if (Variables|Default|GetRemoveFoliageOnSpawn)
    (CallFunction|RemoveFoliageInsideLevelBound)
    (else
      _))
  (Camera|PlayWorldCameraShake (Variables|Default|GetRiftOpeningCameraShake) _returnvalue 0.0 5e+03)
  (ForceFeedback|SpawnForceFeedbackatLocation :ForceFeedbackEffect "/Game/Weapons/ForceFeedback/Explosions/FF_Generic_Explosion_Near.FF_Generic_Explosion_Near" :Location _returnvalue :bLooping false :IntensityMultiplier 1.0 :StartTime 0.0 :AttenuationSettings (Variables|Default|ForceFeedback|GetAttenuationSettings_OnSpawn) :bAutoDestroy true)
  (if (> _spawn_decal_delay_duration 0)
    (Utilities|FlowControl|Delay _spawn_decal_delay_duration)
    (else
      _))
  (if (Variables|Default|GetAllowGroundDecalToSpawn)
    (bind _returnvalue_3 (Rendering|Decal|SpawnDecalatLocation (Variables|Default|GetDecal_GroundBurn) "512.000000,1024.000000,1024.000000" (Transformation|GetActorLocation self)))
    (bind _returnvalue_4 (Rendering|Components|Decal|CreateDynamicMaterialInstance _returnvalue_3))
    (bind (_direction _lightmark _grunge _burnmark) (Timeline|DecalAnimation PlayFromStart)
      (:Update
        (Rendering|Material|SetScalarParameterValue _returnvalue_4 "LightMark0-1" _lightmark :!class /Script/Engine.MaterialInstanceDynamic)
        (Rendering|Material|SetScalarParameterValue _returnvalue_4 "Grunge0-1" _grunge :!class /Script/Engine.MaterialInstanceDynamic)
        (Rendering|Material|SetScalarParameterValue _returnvalue_4 "BurnMark0-1" _burnmark :!class /Script/Engine.MaterialInstanceDynamic)))))

(event Custom|PostSpawnEvents
  ;[1 1 1] Event Used for Children of this Blueprint - Do not remove.
  (CallFunction|AllowSpawnertoSpawn))

(fn GetTeleportLocation
  (bind (_returnvalue_x _returnvalue_y _returnvalue_z) (Math|Vector|BreakVector (Transformation|GetActorLocation self)))
  (return (Math|Vector|MakeVector _returnvalue_x _returnvalue_y (+ _returnvalue_z 1.5e+02))))

(fn OnRep_AllowVFXToSpawn)

(fn OnRep_LevelIsVisble
  (bind _nobuildzone (Variables|Default|GetNoBuildZone))
  (bind _spawnedanticipationvfx (Variables|InternalData|GetSpawnedAnticipationVFX))
  (if (Variables|Default|GetLevelIsVisble)
    (Utilities|IsValid _nobuildzone
      (:Is%20Valid
        (Components|DestroyComponent _nobuildzone) ; Remove the NoBuildZone now that we have spawned in
        (Utilities|IsValid _spawnedanticipationvfx
          (:Is%20Valid
            (Components|Activation|SetActive _spawnedanticipationvfx false false)
            (Components|DestroyComponent _spawnedanticipationvfx)))))))

(fn SpawnAncitipationVFX
  (bind _location (CallFunction|GetTeleportLocation))
  (bind _rift_start_sound (Variables|Sound|GetRift%20Start%20Sound))
  (bind _returnvalue (Niagara|SpawnSystematLocation (Variables|Default|GetFX_Anticipation) _location (Math|Rotator|MakeRotator 0.0 0.0 (Math|Random|RandomFloatinRange -1.8e+02 1.8e+02)) "1.000000,1.000000,1.000000" true true "None" true))
  (Variables|InternalData|SetSpawnedAnticipationVFX _returnvalue)
  (Niagara|SetNiagaraVariable(Float) _returnvalue "Lifetime" (+ (ScalableFloat|GetValueatLevel (Variables|Default|GetAnticipationDelay)) 0.1))
  (LOD|SetMaxDrawDistance _returnvalue 0.0)
  (Audio|SpawnSoundatLocation _rift_start_sound _location)
  (CallFunction|StartAnticipationCameraShake)
  (SoundIndicator|StartSoundIndicatorLooping (Variables|Default|GetFortSoundIndicator) _location _rift_start_sound 1.0 0))

(fn ConstructionScript)
